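# Control Rig for the female medium Sheath player body that adjusts the chest pouch bone to follow upper-body motion.
# The executed chain parent-constrains the pouch base across right clavicle, spine_04 and spine_05 by weight.
# It then rotates the pouch, weighted by a clamped remap of right upper-arm rotation relative to the clavicle.
# Unconnected leftovers include crouch-weighted hip pouch and belt offsets, a belt debug draw and an alternative chest constraint.
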ControlRigBlueprint /Game/Characters/Player/Female/Medium/Bodies/F_Med_Sheath/Meshes/F_MED_Sheath_CtrlRig
graphs [
 RigVMFunctionLibrary {links [], nodes []}
 RigVMModel {nodes [
   n1 = &RigUnit_BeginExecution()
   &RigUnit_ParentConstraint(
    AdvancedSettings=(RotationOrderForFilter=XZY)
    Child=(Type=Bone,Name="dyn_chestPouch_Base")
    Parents=((Item=(Type=Bone,Name="clavicle_r"),Weight=0.1138),(Item=(Type=Bone,Name="spine_04"),Weight=0.4984),(Item=(Type=Bone,Name="spine_05"),Weight=0.3878))
    Weight=1
    ExecutePin=@n1.ExecutePin
   )
   n2 = &RigUnit_GetRelativeTransformForItem(Child=(Type=Bone,Name="upperarm_r"), Parent=(Type=Bone,Name="clavicle_r"))
   > &RigUnit_ModifyTransforms(
    ItemToModify=((Item=(Type=Bone,Name="dyn_chestPouch_Base"),Transform=(Rotation=(X=0,Y=0,Z=0,W=1),Translation=(X=0,Y=0,Z=0),Scale3D=(X=1,Y=1,Z=1))))
    Weight=&RigVMFunction_MathFloatRemap(
     SourceMaximum=85
     SourceMinimum=50
     TargetMinimum=0
     bClamp=true
     Value=&RigVMFunction_MathQuaternionToEuler(RotationOrder=ZXY, Value=@n2.RelativeTransform.Rotation).Result.Z
    ).Result
    ItemToModify.0.Transform.Rotation=&RigVMFunction_MathQuaternionFromEuler(Euler=(X=0,Y=0,Z=-10)).Result
   )
   /Script/ControlRig.RigUnit_ModifyBoneTransforms(
    BoneToModify=((Bone="dyn_hipPouchMain_Base",Transform=(Rotation=(X=0,Y=0,Z=0,W=1),Translation=(X=0,Y=0,Z=0),Scale3D=(X=1,Y=1,Z=1))),(Bone="None",Transform=(Rotation=(X=0,Y=0,Z=0,W=1),Translation=(X=-6,Y=-6,Z=0),Scale3D=(X=1,Y=1,Z=1))))
    ExecutePin=()
    Mode=AdditiveLocal
    Weight=1
    WeightMaximum=1
    WeightMinimum=0
    BoneToModify.0.Transform.Rotation=&RigVMFunction_MathQuaternionFromEuler(Euler=(X=0,Y=-20,Z=0)).Result
   ) {is=struct RigVMUnitNode}
   n3 = RigVMVariableNode() {variable=crouching}
   > &RigUnit_ModifyBoneTransforms(
    BoneToModify=((Bone="dyn_hipPouchMain_Base",Transform=(Rotation=(X=0,Y=0,Z=0,W=1),Translation=(X=-5,Y=-6,Z=5),Scale3D=(X=1,Y=1,Z=1))),(Bone="dyn_belt_Base",Transform=(Rotation=(X=0,Y=0,Z=0,W=1),Translation=(X=0,Y=0,Z=1),Scale3D=(X=1,Y=1,Z=1))))
    Mode=AdditiveLocal
    WeightMaximum=1
    WeightMinimum=0
    BoneToModify.0.Transform.Rotation=&RigVMFunction_MathQuaternionFromEuler(Euler=(X=0,Y=0,Z=-70)).Result
    Weight=@n3.Value
   )
   n4 = &RigUnit_GetBoneTransform(Bone=dyn_belt, Space=LocalSpace)
   > /Script/ControlRig.RigUnit_DebugTransformMutable(
    Color=(R=1,G=0,B=0,A=1)
    DebugDrawSettings=(DepthPriority=SDPG_Foreground,Lifetime=-1)
    Mode=Axes
    Scale=10
    Space=dyn_belt
    Thickness=0
    Transform=(Rotation=(X=0,Y=0,Z=0,W=1),Translation=(X=0,Y=0,Z=0),Scale3D=(X=1,Y=1,Z=1))
    WorldOffset=(Rotation=(X=0,Y=0,Z=0,W=1),Translation=(X=0,Y=0,Z=0),Scale3D=(X=1,Y=1,Z=1))
    bEnabled=True
    Transform.Rotation=/Script/RigVM.RigVMFunction_MathQuaternionFromRotator(Rotator=/Script/RigVM.RigVMFunction_MathQuaternionToRotator(Value=@n4.Transform.Rotation) {is=struct RigVMUnitNode, template="Cast::Execute(in Value,out Result)"}.Result) {is=struct RigVMUnitNode}.Result
   ) {is=struct RigVMUnitNode}
   > &RigUnit_ModifyBoneTransforms(
    BoneToModify=((Bone="dyn_belt_Base",Transform=(Rotation=(X=0,Y=0,Z=0,W=1),Translation=(X=0,Y=0,Z=0),Scale3D=(X=1,Y=1,Z=1))),(Bone="None",Transform=(Rotation=(X=0,Y=0,Z=0,W=1),Translation=(X=0,Y=0,Z=1),Scale3D=(X=1,Y=1,Z=1))))
    Mode=AdditiveLocal
    WeightMaximum=1
    WeightMinimum=0
    BoneToModify.0.Transform.Rotation=&RigVMFunction_MathQuaternionFromEuler(Euler=(X=0,Y=14,Z=0)).Result
    Weight=@n3.Value
   )
   /Script/ControlRig.RigUnit_TransformConstraint(
    BaseBone=None
    BaseTransform=(Rotation=(X=0,Y=0,Z=0,W=1),Translation=(X=0,Y=0,Z=0),Scale3D=(X=1,Y=1,Z=1))
    BaseTransformSpace=GlobalSpace
    Bone=dyn_chestPouch_Base
    ExecutePin=()
    Targets="((Transform=(Rotation=(X=0,Y=0,Z=0,W=1),Translation=(X=0,Y=0,Z=0),Scale3D=(X=1,Y=1,Z=1)),Weight=0.1138,bMaintainOffset=True,Filter=(TranslationFilter=(bX=True,bY=True,bZ=True),RotationFilter=(bX=True,bY=True,bZ=True),ScaleFilter=(bX=True,bY=True,bZ=True))) ...(truncated)"
    bUseInitialTransforms=True
    Targets.0.Transform=&RigUnit_GetBoneTransform(Bone=clavicle_r).Transform
    Targets.1.Transform=&RigUnit_GetBoneTransform(Bone=spine_04).Transform
    Targets.2.Transform=&RigUnit_GetBoneTransform(Bone=spine_05).Transform
   ) {is=struct RigVMUnitNode}
   > &RigUnit_ModifyBoneTransforms(
    BoneToModify=((Bone="dyn_chestPouch_Base",Transform=(Rotation=(X=0,Y=0,Z=0,W=1),Translation=(X=0,Y=0,Z=-2),Scale3D=(X=1,Y=1,Z=1))),(Bone="None",Transform=(Rotation=(X=0,Y=0,Z=0,W=1),Translation=(X=0,Y=0,Z=1),Scale3D=(X=1,Y=1,Z=1))))
    Mode=AdditiveLocal
    Weight=0
    WeightMaximum=1
    WeightMinimum=0
   )
   &RigUnit_GetRelativeBoneTransform(Bone=upperarm_r, Space=clavicle_r)
   &RigVMFunction_MathFloatRemap(
    SourceMaximum=-106
    SourceMinimum=0
    TargetMinimum=0.1
    bClamp=true
    Value=&RigVMFunction_MathQuaternionToEuler(RotationOrder=XYZ, Value=&RigUnit_GetRelativeBoneTransform(Bone=thigh_l, Space=pelvis).Transform.Rotation).Result.Z
   )
  ]}
]
hierarchy [
 root {loc=<0 0 0>, rot=<0 0 0 1>}
 pelvis {parent=root, loc=<0 -0.5301 98.0571>, rot=<0.0127 0.707 -0.0127 -0.707>}
 spine_01 {parent=pelvis, loc=<6.1208 -0.9328 0>, rot=<0 0 0.034 -0.9994>}
 spine_02 {parent=spine_01, loc=<6.5996 -0.2095 0>, rot=<0 0 0.0238 -0.9997>}
 spine_03 {parent=spine_02, loc=<6.4914 -0.1237 0>, rot=<0 0 0.0149 -0.9999>}
 spine_04 {parent=spine_03, loc=<6.9598 0.0794 0>, rot=<0 0 -0.0132 -0.9999>}
 spine_05 {parent=spine_04, loc=<6.9987 0 0>, rot=<0 0 -0.0015 -1>}
 clavicle_l {parent=spine_05, loc=<5.4831 -0.1053 -3.2946>, rot=<0.0465 -0.7289 -0.009 -0.6829>}
 upperarm_l {parent=clavicle_l, loc=<9.8658 -1.2136 -0.4104>, rot=<0.0164 -0.3245 -0.0269 -0.9454>}
 lowerarm_l {parent=upperarm_l, loc=<26.465 -0.8345 0.259>, rot=<0.0036 0.0161 -0.2298 -0.9731>}
 hand_l {parent=lowerarm_l, loc=<22.7355 -0.3203 -0.6105>, rot=<-0.6784 -0.0223 0.0735 -0.7307>}
 index_metacarpal_l {parent=hand_l, loc=<1.6535 0.0183 -2.96>, rot=<0.0158 -0.0124 -0.0173 -0.9996>}
 index_01_l {parent=index_metacarpal_l, loc=<8.3008 0.0357 -0.0524>, rot=<-0.0044 0.0485 0.0924 -0.9945>}
 index_02_l {parent=index_01_l, loc=<3.3587 -0.3004 0.0452>, rot=<-0.002 0.0027 0.3225 -0.9466>}
 index_03_l {parent=index_02_l, loc=<2.5578 -0.1231 -0.0252>, rot=<-0.002 0.0067 0.2558 -0.9667>}
 middle_metacarpal_l {parent=hand_l, loc=<2.1856 -0.1725 -0.5984>, rot=<0.006 0.0168 -0.0329 -0.9993>}
 middle_01_l {parent=middle_metacarpal_l, loc=<7.3275 -0.2447 0.0129>, rot=<-0.0208 0.0042 0.0858 -0.9961>}
 middle_02_l {parent=middle_01_l, loc=<4.0123 0.1401 0.1175>, rot=<-0.0218 0.0258 0.3512 -0.9357>}
 middle_03_l {parent=middle_02_l, loc=<2.6077 0.0534 0.1373>, rot=<0.0082 0.0188 0.1708 -0.9851>}
 pinky_metacarpal_l {parent=hand_l, loc=<1.7616 -0.9753 3.3814>, rot=<-0.0397 0.0814 -0.069 -0.9935>}
 pinky_01_l {parent=pinky_metacarpal_l, loc=<7.1709 -0.4831 -0.1094>, rot=<-0.0654 -0.0035 0.2242 -0.9723>}
 pinky_02_l {parent=pinky_01_l, loc=<2.8274 0.0893 -0.1043>, rot=<0.0037 0.0011 0.2854 -0.9584>}
 pinky_03_l {parent=pinky_02_l, loc=<2.2046 0.0308 -0.0078>, rot=<-0.0002 0.0098 0.1445 -0.9895>}
 ring_metacarpal_l {parent=hand_l, loc=<2.2074 -0.3625 1.3638>, rot=<-0.0284 0.0722 -0.0458 -0.9959>}
 ring_01_l {parent=ring_metacarpal_l, loc=<7.0295 -0.6197 -0.1446>, rot=<-0.0489 -0.0252 0.1892 -0.9804>}
 ring_02_l {parent=ring_01_l, loc=<3.6174 0.4647 -0.1027>, rot=<0 0 0.2806 -0.9598>}
 ring_03_l {parent=ring_02_l, loc=<2.4152 0.1141 0.0199>, rot=<0 0 0.1189 -0.9929>}
 thumb_01_l {parent=hand_l, loc=<1.3118 -0.3191 -2.459>, rot=<0.6546 -0.4042 0.0664 -0.6354>}
 thumb_02_l {parent=thumb_01_l, loc=<5.9336 -0.0448 0.0401>, rot=<-0.0037 0.0213 0.298 -0.9543>}
 thumb_03_l {parent=thumb_02_l, loc=<2.2223 0.3682 -0.0907>, rot=<0.0118 -0.013 0.0066 -0.9998>}
 lowerarm_twist_01_l {parent=lowerarm_l, loc=<18.316 0 0>, rot=<0 0 0 1>}
 lowerarm_twist_02_l {parent=lowerarm_l, loc=<10.6707 0 0>, rot=<0 0 0 1>}
 upperarm_twist_01_l {parent=upperarm_l, loc=<7.2135 -0.1134 0>, rot=<0 0 0 -1>}
 upperarm_twist_02_l {parent=upperarm_l, loc=<19.7998 -0.1134 0>, rot=<0 0 0 -1>}
 clavicle_r {parent=spine_05, loc=<5.4831 -0.1053 3.2946>, rot=<0.7289 0.0465 -0.6829 0.009>}
 upperarm_r {parent=clavicle_r, loc=<-9.8658 1.2136 0.4104>, rot=<0.0164 -0.3245 -0.0269 -0.9454>}
 lowerarm_r {parent=upperarm_r, loc=<-26.4651 0.8345 -0.259>, rot=<0.0036 0.0161 -0.2298 -0.9731>}
 hand_r {parent=lowerarm_r, loc=<-22.7356 0.3203 0.6106>, rot=<-0.6784 -0.0223 0.0735 -0.7307>}
 index_metacarpal_r {parent=hand_r, loc=<-1.6534 -0.0181 2.9601>, rot=<0.0158 -0.0124 -0.0173 -0.9996>}
 index_01_r {parent=index_metacarpal_r, loc=<-8.3008 -0.0357 0.0524>, rot=<-0.0044 0.0485 0.0924 -0.9945>}
 index_02_r {parent=index_01_r, loc=<-3.3589 0.3003 -0.0453>, rot=<-0.002 0.0027 0.3225 -0.9466>}
 index_03_r {parent=index_02_r, loc=<-2.5578 0.123 0.0252>, rot=<-0.002 0.0067 0.2558 -0.9667>}
 middle_metacarpal_r {parent=hand_r, loc=<-2.1856 0.1726 0.5984>, rot=<0.006 0.0168 -0.0329 -0.9993>}
 middle_01_r {parent=middle_metacarpal_r, loc=<-7.3273 0.2447 -0.0129>, rot=<-0.0208 0.0042 0.0858 -0.9961>}
 middle_02_r {parent=middle_01_r, loc=<-4.0124 -0.14 -0.1175>, rot=<-0.0218 0.0258 0.3512 -0.9357>}
 middle_03_r {parent=middle_02_r, loc=<-2.6077 -0.0533 -0.1373>, rot=<0.0082 0.0188 0.1708 -0.9851>}
 pinky_metacarpal_r {parent=hand_r, loc=<-1.7617 0.9753 -3.3814>, rot=<-0.0397 0.0814 -0.069 -0.9935>}
 pinky_01_r {parent=pinky_metacarpal_r, loc=<-7.1707 0.4832 0.1094>, rot=<-0.0654 -0.0035 0.2242 -0.9723>}
 pinky_02_r {parent=pinky_01_r, loc=<-2.8274 -0.0892 0.1043>, rot=<0.0037 0.0011 0.2854 -0.9584>}
 pinky_03_r {parent=pinky_02_r, loc=<-2.2045 -0.031 0.0078>, rot=<-0.0002 0.0098 0.1445 -0.9895>}
 ring_metacarpal_r {parent=hand_r, loc=<-2.2073 0.3627 -1.3638>, rot=<-0.0284 0.0722 -0.0458 -0.9959>}
 ring_01_r {parent=ring_metacarpal_r, loc=<-7.0297 0.6197 0.1446>, rot=<-0.0489 -0.0252 0.1892 -0.9804>}
 ring_02_r {parent=ring_01_r, loc=<-3.6171 -0.4648 0.1027>, rot=<0 0 0.2806 -0.9598>}
 ring_03_r {parent=ring_02_r, loc=<-2.4153 -0.1141 -0.0199>, rot=<0 0 0.1189 -0.9929>}
 thumb_01_r {parent=hand_r, loc=<-1.3119 0.3192 2.459>, rot=<0.6546 -0.4042 0.0664 -0.6354>}
 thumb_02_r {parent=thumb_01_r, loc=<-5.9336 0.0447 -0.0401>, rot=<-0.0037 0.0213 0.298 -0.9543>}
 thumb_03_r {parent=thumb_02_r, loc=<-2.2223 -0.3682 0.0907>, rot=<0.0118 -0.013 0.0066 -0.9998>}
 lowerarm_twist_01_r {parent=lowerarm_r, loc=<-18.3161 0 0>, rot=<0 0 0 1>}
 lowerarm_twist_02_r {parent=lowerarm_r, loc=<-10.6707 0 0>, rot=<0 0 0 1>}
 upperarm_twist_01_r {parent=upperarm_r, loc=<-7.2135 0.1134 0>, rot=<0 0 0 1>}
 dyn_armPouchMain_Base_r {parent=upperarm_twist_01_r, loc=<-3.8088 0.3053 -2.8>, rot=<-0.4901 0.5067 0.5134 0.4893>}
 dyn_armPouchMain_r {parent=dyn_armPouchMain_Base_r, loc=<0 0 0>, rot=<0 0 0 -1>}
 dyn_armPouchStrap_r {parent=dyn_armPouchMain_r, loc=<5.1838 3.7662 0>, rot=<-0.0012 0.0086 -0.6145 -0.7889>}
 upperarm_twist_02_r {parent=upperarm_r, loc=<-19.7998 0.1134 0>, rot=<0 0 0 1>}
 dyn_chestPouch_Base {parent=clavicle_r, loc=<-4.8083 -7.9134 2.6412>, rot=<-0.6004 0.3593 0.6129 -0.367>}
 dyn_chestPouch {parent=dyn_chestPouch_Base, loc=<0 0 0>, rot=<0 0 0 -1>}
 dyn_chestPouchStrap {parent=dyn_chestPouch, loc=<5.5038 4.161 0.0301>, rot=<0.028 0.0418 -0.5509 -0.8331>}
 dyn_chestPouchDRing {parent=dyn_chestPouchStrap, loc=<1.9946 0.0729 0.0019>, rot=<0 0 0 1>}
 dyn_chestPouchZipper_Base {parent=dyn_chestPouch, loc=<1.8596 -1.3087 2.9197>, rot=<0.5736 0.4971 -0.4191 -0.4982>}
 dyn_chestPouchZipper {parent=dyn_chestPouchZipper_Base, loc=<0 0 0>, rot=<0 0 0 1>}
 neck_01 {parent=spine_05, loc=<11.3138 0.1529 0>, rot=<0 0 -0.2003 -0.9797>}
 neck_02 {parent=neck_01, loc=<4.5627 0 0>, rot=<0 0 0.0504 -0.9987>}
 head {parent=neck_02, loc=<5.3914 0 0>, rot=<0 0 0.115 -0.9934>}
 dyn_neckZipper_Base {parent=neck_01, loc=<1.8091 5.3287 0.0202>, rot=<0.0009 0.0064 0.9832 0.1823>}
 dyn_neckZipper {parent=dyn_neckZipper_Base, loc=<0 0 0>, rot=<0 0 0 1>}
 dyn_belt_Base {parent=spine_01, loc=<3.926 0.0929 8.2859>, rot=<0.5 -0.5 -0.5 -0.5>}
 dyn_belt {parent=dyn_belt_Base, loc=<0 0 0>, rot=<0 0 0 -1>}
 dyn_beltEND {parent=dyn_belt, loc=<18.0766 2.813 5.1652>, rot=<-0.6648 0.2048 -0.2729 -0.6645>}
 dyn_hipPouchMain_Base {parent=spine_01, loc=<1.113 5.258 -9.7907>, rot=<-0.239 -0.6987 -0.2387 -0.6306>}
 dyn_hipPouchMain {parent=dyn_hipPouchMain_Base, loc=<0 0 0>, rot=<0 0 0 -1>}
 dyn_hipGrenade {parent=dyn_hipPouchMain, loc=<4.3487 4.8686 -2.9216>, rot=<0.574 0.4552 -0.5052 -0.456>}
 dyn_grenadePin {parent=dyn_hipGrenade, loc=<2.252 -5.9228 1.0624>, rot=<-0.1054 -0.5267 -0.379 -0.7535>}
 dyn_hipPouchStrap {parent=dyn_hipPouchMain, loc=<8.7048 -0.9437 -4.8666>, rot=<0.6682 -0.3909 -0.31 -0.552>}
 dyn_hipPouchDRing {parent=dyn_hipPouchStrap, loc=<2.7355 0.1 0>, rot=<0 0 0 1>}
 thigh_l {parent=pelvis, loc=<-1.0855 -0.2109 -8.7053>, rot=<0.0009 0.0355 0.0155 -0.9993>}
 calf_l {parent=thigh_l, loc=<-42.5284 -0.0019 -0.7297>, rot=<0.0006 -0.0161 -0.0391 -0.9991>}
 calf_twist_01_l {parent=calf_l, loc=<-35.8483 -1.9197 -0.5794>, rot=<0.0237 0.0095 -0.0035 -0.9997>}
 calf_twist_02_l {parent=calf_l, loc=<-19.1546 -1.7943 -0.2638>, rot=<0.0237 0.0095 -0.0035 -0.9997>}
 foot_l {parent=calf_l, loc=<-46.7401 -2.5014 -0.592>, rot=<0.0427 -0.0238 0.0353 -0.9982>}
 ball_l {parent=foot_l, loc=<-3.5658 10.8066 -0.1544>, rot=<0.0001 -0.0001 -0.7063 -0.7079>}
 thigh_twist_01_l {parent=thigh_l, loc=<-27.952 0 0>, rot=<0.0013 0 0 -1>}
 sub_thigh_l {parent=thigh_l, loc=<-2.972 0 0>, rot=<0 0 0 -1>}
 driver_frill_frt_l {parent=sub_thigh_l, loc=<1.8024 8.4707 -0.9042>, rot=<0 0 0 -1>}
 dyn_frill_frt_l {parent=driver_frill_frt_l, loc=<0 0 0>, rot=<0 0 0 -1>}
 driver_frill_side_l {parent=sub_thigh_l, loc=<2.8149 -0.4685 -5.5899>, rot=<0 0 0 -1>}
 dyn_frill_side_l {parent=driver_frill_side_l, loc=<0 0 0>, rot=<0 0 0 -1>}
 driver_frill_bck_l {parent=sub_thigh_l, loc=<3.0963 -8.8443 -1.1757>, rot=<0 0 0 -1>}
 dyn_frill_bck_l {parent=driver_frill_bck_l, loc=<0 0 0>, rot=<0 0 0 -1>}
 thigh_r {parent=pelvis, loc=<-1.0853 -0.2109 8.7053>, rot=<0.0355 -0.0009 0.9993 0.0155>}
 calf_r {parent=thigh_r, loc=<42.5286 0.0019 0.7297>, rot=<0.0006 -0.0161 -0.0391 -0.9991>}
 calf_twist_01_r {parent=calf_r, loc=<35.8485 1.9197 0.5794>, rot=<0.0237 0.0095 -0.0035 -0.9997>}
 calf_twist_02_r {parent=calf_r, loc=<19.1546 1.7943 0.2638>, rot=<0.0237 0.0095 -0.0035 -0.9997>}
 foot_r {parent=calf_r, loc=<46.7402 2.5014 0.592>, rot=<0.0427 -0.0238 0.0353 -0.9982>}
 ball_r {parent=foot_r, loc=<3.5658 -10.8066 0.1545>, rot=<0.0001 -0.0001 -0.7063 -0.7079>}
 thigh_twist_01_r {parent=thigh_r, loc=<27.952 0 0>, rot=<0.0013 0 0 -1>}
 sub_thigh_r {parent=thigh_r, loc=<2.9722 0 0>, rot=<0 0 0 -1>}
 driver_frill_frt_r {parent=sub_thigh_r, loc=<-1.8025 -8.4707 0.9042>, rot=<0 0 0 -1>}
 dyn_frill_frt_r {parent=driver_frill_frt_r, loc=<0 0 0>, rot=<0 0 0 -1>}
 driver_frill_side_r {parent=sub_thigh_r, loc=<-2.815 0.4685 5.59>, rot=<0 0 0 -1>}
 dyn_frill_side_r {parent=driver_frill_side_r, loc=<0 0 0>, rot=<0 0 0 -1>}
 driver_frill_bck_r {parent=sub_thigh_r, loc=<-3.0964 8.8443 1.1757>, rot=<0 0 0 -1>}
 dyn_frill_bck_r {parent=driver_frill_bck_r, loc=<0 0 0>, rot=<0 0 0 -1>}
 driver_wire {parent=thigh_r, loc=<-3.3316 -4.5949 3.6219>, rot=<0.9994 -0.0024 -0.0355 -0.0003>}
 dyn_wire {parent=driver_wire, loc=<0 0 0>, rot=<0 0 0 -1>}
 dyn_wire_front {parent=dyn_wire, loc=<4.7089 5.4288 2.3878>, rot=<0.9994 -0.0024 -0.0355 0.0003>}
 dyn_wire_back {parent=dyn_wire, loc=<3.11 -7.3715 -3.439>, rot=<0.9994 -0.0024 -0.0355 0.0003>}
 driver_frill_bck {parent=pelvis, loc=<-1.4605 -10.6717 0>, rot=<-0.0127 -0.707 0.0127 -0.707>}
 dyn_frill_bck {parent=driver_frill_bck, loc=<0 0 0>, rot=<0 0 0 1>}
 Flap_Front_Parent_Space {parent=pelvis, loc=<-2.6394 10.9147 0>, rot=<-0.0127 -0.707 0.0127 -0.707>}
 Flap_Front_Driven_Bone {parent=Flap_Front_Parent_Space, loc=<0 0 0>, rot=<0 0 0 1>}
 dyn_frill_frt {parent=Flap_Front_Driven_Bone, loc=<0 0 0>, rot=<0 0 0 1>}
]
variables [crouching {type=double, default=0}]
