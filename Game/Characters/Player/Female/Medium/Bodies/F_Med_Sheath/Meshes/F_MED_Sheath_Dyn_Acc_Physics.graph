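# Physics asset for the dynamic accessories on the medium female Sheath body: arm, chest and hip pouches, belt, grenade with pin, and zippers.
# Chest pouch parts use boxes; all other bodies, including pelvis, left thigh and frill driver, use capsules.
# Constraints form short chains, mostly rooted at pelvis through free joints, though the hip pouch base has no parent constraint.
# Straps, D-rings, zippers, pin and belt swing within limits, often with locked axes.
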PhysicsAsset /Game/Characters/Player/Female/Medium/Bodies/F_Med_Sheath/Meshes/F_MED_Sheath_Dyn_Acc_Physics
bodies {
 driver_frill_frt_l [{shapeType=Capsule, length=6.005, radius=7.255, center=<-2.5393 -4.4963 4.1549>, rotation=<rot 56.6083 -123.5495 63.2212>}]
 dyn_armPouchMain_r [{shapeType=Capsule, length=4.8928, radius=3.5173, center=<3.4763 0.727 -0.0098>, rotation=<rot -0.3639 -2.3545 -88.8249>}]
 dyn_armPouchStrap_r [{shapeType=Capsule, length=2.4139, radius=1.1082, center=<1.7375 0 0.0001>, rotation=<rot 90 0 0>}]
 dyn_belt [{shapeType=Capsule, length=11.9369, radius=3.053, center=<5.8552 0 0>, rotation=<rot 90 0 0>}]
 dyn_belt_Base [{shapeType=Capsule, length=0.505, radius=0.505, rotation=<rot 0 -90 0>}]
 dyn_chestPouch [{shapeType=Box, extentX=3.3571, extentY=5.3236, extentZ=12.551, center=<2.2201 -0.0793 -0.2667>, rotation=<rot -4.9796 168.4206 -89.2373>}]
 dyn_chestPouchDRing [{shapeType=Box, extentX=1.01, extentY=1.7332, extentZ=2.6323, center=<0.7382 0 0>, rotation=<rot 0 -90 0>}]
 dyn_chestPouchStrap [{shapeType=Box, extentX=1.01, extentY=5.1344, extentZ=1.8698, center=<2.0213 -0.0001 0>, rotation=<rot 0 -90 0>}]
 dyn_chestPouchZipper [{shapeType=Box, extentX=1.01, extentY=4.937, extentZ=1.01, center=<1.6235 0 0>, rotation=<rot 0 -90 0>}]
 dyn_chestPouchZipper_Base [{shapeType=Box, extentX=1.01, extentY=1.01, extentZ=1.01, rotation=<rot 0 -90 0>}]
 dyn_chestPouch_Base [{shapeType=Box, extentX=1.1897, extentY=2.976, extentZ=8.9752, center=<-0.0585 -1.0634 0.0719>, rotation=<rot 0 -10.7871 -89.1783>}]
 dyn_grenadePin [{shapeType=Capsule, length=1.8697, radius=1.8651, center=<1.6552 0.0377 -0.1426>, rotation=<rot 0 87.5828 -91.4728>}]
 dyn_hipGrenade [{shapeType=Capsule, length=7.2548, radius=3.1721, center=<2.2074 0.1894 0.1303>, rotation=<rot 0 1.4048 -78.3816>}]
 dyn_hipPouchDRing [{shapeType=Capsule, length=1.7073, radius=1.2262, center=<0.8241 -0.0274 -0.0173>, rotation=<rot -0.6702 -90 -1.4216>}]
 dyn_hipPouchMain [{shapeType=Capsule, length=7.1904, radius=5.4454, center=<4.2543 -0.3549 -1.6581>, rotation=<rot 6.4757 -90 -18.4762>}]
 dyn_hipPouchMain_Base [{shapeType=Capsule, length=0.505, radius=0.505, rotation=<rot 0 -90 0>}]
 dyn_hipPouchStrap [{shapeType=Capsule, length=2.2277, radius=1.2735, center=<1.6859 0.1189 -0.029>, rotation=<rot -0.2827 -90 -93.0126>}]
 dyn_neckZipper [{shapeType=Capsule, length=2.7085, radius=0.505, center=<1.3451 0 0.0061>, rotation=<rot 0 -90.0057 -89.8595>}]
 dyn_neckZipper_Base [{shapeType=Capsule, length=0.505, radius=0.505, rotation=<rot 0 -90 0>}]
 pelvis [{shapeType=Capsule, length=7.8103, radius=8.9351, center=<5.1782 0 0>, rotation=<rot 90 0 0>}]
 sub_thigh_l [{shapeType=Capsule, length=0.505, radius=0.505, rotation=<rot 0 -90 180>}]
}
constraints [
 {
  bone1Name=sub_thigh_l
  bone2Name=pelvis
  swing1LimitDegrees=45
  swing1Motion=Free
  swing2LimitDegrees=45
  swing2Motion=Free
  twistLimitDegrees=45
  twistMotion=Free
 }
 {
  bone1Name=driver_frill_frt_l
  bone2Name=sub_thigh_l
  swing1LimitDegrees=45
  swing1Motion=Free
  swing2LimitDegrees=45
  swing2Motion=Free
  twistLimitDegrees=45
  twistMotion=Free
 }
 {
  bone1Name=dyn_armPouchMain_r
  bone2Name=pelvis
  swing1LimitDegrees=45
  swing1Motion=Free
  swing2LimitDegrees=45
  swing2Motion=Free
  twistLimitDegrees=45
  twistMotion=Free
 }
 {
  bone1Name=dyn_armPouchStrap_r
  bone2Name=dyn_armPouchMain_r
  swing1LimitDegrees=30
  swing2LimitDegrees=8
  twistLimitDegrees=45
  twistMotion=Locked
 }
 {
  bone1Name=dyn_neckZipper_Base
  bone2Name=pelvis
  swing1LimitDegrees=45
  swing1Motion=Free
  swing2LimitDegrees=45
  swing2Motion=Free
  twistLimitDegrees=45
  twistMotion=Free
 }
 {
  bone1Name=dyn_neckZipper
  bone2Name=dyn_neckZipper_Base
  swing1LimitDegrees=45
  swing2LimitDegrees=45
  swing2Motion=Locked
  twistLimitDegrees=45
  twistMotion=Locked
 }
 {
  bone1Name=dyn_belt_Base
  bone2Name=pelvis
  swing1LimitDegrees=45
  swing1Motion=Free
  swing2LimitDegrees=45
  swing2Motion=Free
  twistLimitDegrees=45
  twistMotion=Free
 }
 {bone1Name=dyn_hipPouchMain, bone2Name=dyn_hipPouchMain_Base, swing1LimitDegrees=8, swing2LimitDegrees=8, twistLimitDegrees=8}
 {
  bone1Name=dyn_hipGrenade
  bone2Name=dyn_hipPouchMain
  swing1LimitDegrees=45
  swing1Motion=Locked
  swing2LimitDegrees=45
  swing2Motion=Locked
  twistLimitDegrees=10
 }
 {
  bone1Name=dyn_grenadePin
  bone2Name=dyn_hipGrenade
  swing1LimitDegrees=32
  swing2LimitDegrees=45
  swing2Motion=Locked
  twistLimitDegrees=45
  twistMotion=Locked
 }
 {
  bone1Name=dyn_hipPouchStrap
  bone2Name=dyn_hipPouchMain
  swing1LimitDegrees=30
  swing2LimitDegrees=10
  twistLimitDegrees=45
  twistMotion=Locked
 }
 {
  bone1Name=dyn_hipPouchDRing
  bone2Name=dyn_hipPouchStrap
  swing1LimitDegrees=45
  swing2LimitDegrees=45
  swing2Motion=Locked
  twistLimitDegrees=45
  twistMotion=Locked
 }
 {
  bone1Name=dyn_belt
  bone2Name=dyn_belt_Base
  swing1LimitDegrees=12
  swing2LimitDegrees=5
  swing2Motion=Locked
  twistLimitDegrees=5
  twistMotion=Locked
 }
 {
  bone1Name=dyn_chestPouch_Base
  bone2Name=pelvis
  swing1LimitDegrees=45
  swing1Motion=Free
  swing2LimitDegrees=45
  swing2Motion=Free
  twistLimitDegrees=45
  twistMotion=Free
 }
 {
  bone1Name=dyn_chestPouch
  bone2Name=dyn_chestPouch_Base
  swing1LimitDegrees=45
  swing1Motion=Free
  swing2LimitDegrees=45
  swing2Motion=Free
  twistLimitDegrees=45
  twistMotion=Free
 }
 {bone1Name=dyn_chestPouchStrap, bone2Name=dyn_chestPouch, swing1LimitDegrees=25, swing2LimitDegrees=5, twistLimitDegrees=5}
 {
  bone1Name=dyn_chestPouchDRing
  bone2Name=dyn_chestPouchStrap
  swing1LimitDegrees=45
  swing2LimitDegrees=45
  swing2Motion=Locked
  twistLimitDegrees=45
  twistMotion=Locked
 }
 {
  bone1Name=dyn_chestPouchZipper_Base
  bone2Name=dyn_chestPouch
  swing1LimitDegrees=45
  swing1Motion=Free
  swing2LimitDegrees=45
  swing2Motion=Free
  twistLimitDegrees=45
  twistMotion=Free
 }
 {
  bone1Name=dyn_chestPouchZipper
  bone2Name=dyn_chestPouchZipper_Base
  swing1LimitDegrees=45
  swing2LimitDegrees=45
  swing2Motion=Locked
  twistLimitDegrees=45
  twistMotion=Locked
 }
]
