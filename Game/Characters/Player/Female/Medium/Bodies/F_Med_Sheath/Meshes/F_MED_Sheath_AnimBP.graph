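# Animation Blueprint for the Sheath female medium player body, adding secondary motion to skirt frills, wires, belt, pouches and zipper.
# It copies the attached parent mesh's pose, then uses constraints and additive bone offsets driven by leg-raise, leg-cross, jump and skydive alphas.
# Skirt and accessory RigidBody passes use separate physics assets, with spring AnimDynamics on the wires between them.
# A final Control Rig receives the crouch alpha; the EventGraph is empty.
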AnimBlueprint /Game/Characters/Player/Female/Medium/Bodies/F_Med_Sheath/Meshes/F_MED_Sheath_AnimBP
target_skeleton=&/F_MED_Sheath_Skeleton
graphs [
 AnimGraph {nodes [
   n1 = "Constraint - dyn_chestPouch_Base (PPP)"!(
    ComponentPose=Local To Component(LocalPose=Copy Pose From Mesh() {bUseAttachedParent=true})
    Alpha=1
    ConstraintWeights_0=0.1138
    ConstraintWeights_1:float=0.4984
    ConstraintWeights_2:float=0.3878
   ) -> (Pose) {
    alphaBoolBlend {blendOption=Linear}
    boneToModify {boneName=dyn_chestPouch_Base}
    constraintSetup [{targetBone {boneName=clavicle_r}}, {targetBone {boneName=spine_04}}, {targetBone {boneName=spine_05}}]
    constraintWeights [1,1,1]
   }
   n2 = "Constraint - sub_thigh_l (P)"(ComponentPose=@n1, Alpha=1, ConstraintWeights_0=1) {
    alphaBoolBlend {blendOption=Linear}
    boneToModify {boneName=sub_thigh_l}
    constraintSetup [{targetBone {boneName=thigh_twist_01_l}}]
    constraintWeights [1]
   }
   n3 = "Constraint - sub_thigh_r (P)"(ComponentPose=@n2, Alpha=1, ConstraintWeights_0=1) {
    alphaBoolBlend {blendOption=Linear}
    boneToModify {boneName=sub_thigh_r}
    constraintSetup [{targetBone {boneName=thigh_twist_01_r}}]
    constraintWeights [1]
   }
   n4 = Get LegUpLeftAlpha() -> (LegUpLeftAlpha:float)
   n5 = "Transform (Modify) Bone - Bone: driver_frill_frt_l"(ComponentPose=@n3, Alpha=@n4, Translation="0,1.5,0", Rotation="-3,0,0", Scale="1,1,1") {
    rotationMode=BMM_Additive
    rotationSpace=BCS_BoneSpace
    translationMode=BMM_Additive
    translationSpace=BCS_BoneSpace
    alphaBoolBlend {blendOption=Linear}
    alphaScaleBiasClamp {bMapRange=true, inRange {max=0, min=-100}, outRange {max=0, min=1}}
    boneToModify {boneName=driver_frill_frt_l}
    rotation {pitch=-3}
    translation {y=1.5}
   }
   n6 = "Transform (Modify) Bone - Bone: driver_frill_side_l"(ComponentPose=@n5, Alpha=@n4, Translation="0,0,-2", Rotation="0,0,0", Scale="1,1,1") {
    rotationMode=BMM_Additive
    rotationSpace=BCS_BoneSpace
    translationMode=BMM_Additive
    translationSpace=BCS_BoneSpace
    alphaBoolBlend {blendOption=Linear}
    alphaScaleBiasClamp {bMapRange=true, inRange {max=0, min=-100}, outRange {max=0, min=1}}
    boneToModify {boneName=driver_frill_side_l}
    translation {z=-2}
   }
   n7 = "Transform (Modify) Bone - Bone: driver_frill_bck_l"(ComponentPose=@n6, Alpha=@n4, Translation="5,0,0", Rotation="0,80,0", Scale="1,1,1") {
    rotationMode=BMM_Additive
    rotationSpace=BCS_BoneSpace
    translationMode=BMM_Additive
    translationSpace=BCS_BoneSpace
    alphaBoolBlend {blendOption=Linear}
    alphaScaleBiasClamp {bMapRange=true, inRange {max=0, min=-100}, outRange {max=0, min=1}}
    boneToModify {boneName=driver_frill_bck_l}
    rotation {yaw=80}
    translation {x=5}
   }
   n8 = Get LegUpRightAlpha() -> (LegUpRightAlpha:float)
   n9 = "Transform (Modify) Bone - Bone: driver_frill_frt_r"(ComponentPose=@n7, Alpha=@n8, Translation="0,-1,0", Rotation="-3,0,0", Scale="1,1,1") {
    rotationMode=BMM_Additive
    rotationSpace=BCS_BoneSpace
    translationMode=BMM_Additive
    translationSpace=BCS_BoneSpace
    alphaBoolBlend {blendOption=Linear}
    alphaScaleBiasClamp {bMapRange=true, inRange {max=0, min=-100}, outRange {max=0, min=1}}
    boneToModify {boneName=driver_frill_frt_r}
    rotation {pitch=-3}
    translation {y=-1}
   }
   n10 = "Transform (Modify) Bone - Bone: driver_frill_side_r"(ComponentPose=@n9, Alpha=@n8, Translation="0,0,2", Rotation="0,0,0", Scale="1,1,1") {
    rotationMode=BMM_Additive
    rotationSpace=BCS_BoneSpace
    translationMode=BMM_Additive
    translationSpace=BCS_BoneSpace
    alphaBoolBlend {blendOption=Linear}
    alphaScaleBiasClamp {bMapRange=true, inRange {max=0, min=-100}, outRange {max=0, min=1}}
    boneToModify {boneName=driver_frill_side_r}
    translation {z=2}
   }
   n11 = "Transform (Modify) Bone - Bone: driver_frill_bck_r"(ComponentPose=@n10, Alpha=@n8, Translation="-5,0,0", Rotation="0,80,0", Scale="1,1,1") {
    rotationMode=BMM_Additive
    rotationSpace=BCS_BoneSpace
    translationMode=BMM_Additive
    translationSpace=BCS_BoneSpace
    alphaBoolBlend {blendOption=Linear}
    alphaScaleBiasClamp {bMapRange=true, inRange {max=0, min=-100}, outRange {max=0, min=1}}
    boneToModify {boneName=driver_frill_bck_r}
    rotation {yaw=80}
    translation {x=-5}
   }
   n12 = Get RightLeftCrossAlpha() -> (RightLeftCrossAlpha:float)
   n13 = "Transform (Modify) Bone - Bone: driver_frill_bck_l"(ComponentPose=@n11, Alpha=@n12, Translation="0,0,0", Rotation="30,0,0", Scale="1,1,1") {
    rotationMode=BMM_Additive
    rotationSpace=BCS_BoneSpace
    translationMode=BMM_Additive
    translationSpace=BCS_BoneSpace
    alphaBoolBlend {blendOption=Linear}
    alphaScaleBiasClamp {inRange {max=0, min=-100}, outRange {max=0, min=1}}
    boneToModify {boneName=driver_frill_bck_l}
    rotation {pitch=30}
   }
   n14 = "Transform (Modify) Bone - Bone: driver_frill_bck_r"(ComponentPose=@n13, Alpha=@n12, Translation="0,0,0", Rotation="30,0,0", Scale="1,1,1") {
    rotationMode=BMM_Additive
    rotationSpace=BCS_BoneSpace
    translationMode=BMM_Additive
    translationSpace=BCS_BoneSpace
    alphaBoolBlend {blendOption=Linear}
    alphaScaleBiasClamp {inRange {max=0, min=-100}, outRange {max=0, min=1}}
    boneToModify {boneName=driver_frill_bck_r}
    rotation {pitch=30}
   }
   n15 = Get LeftLegCrossAlpha() -> (LeftLegCrossAlpha:float)
   n16 = "Transform (Modify) Bone - Bone: driver_frill_bck_l"(ComponentPose=@n14, Alpha=@n15, Translation="0,0,0", Rotation="40,0,0", Scale="1,1,1") {
    rotationMode=BMM_Additive
    rotationSpace=BCS_BoneSpace
    translationMode=BMM_Additive
    translationSpace=BCS_BoneSpace
    alphaBoolBlend {blendOption=Linear}
    alphaScaleBiasClamp {inRange {max=0, min=-100}, outRange {max=0, min=1}}
    boneToModify {boneName=driver_frill_bck_l}
    rotation {pitch=40}
   }
   n17 = "Transform (Modify) Bone - Bone: driver_frill_bck_r"(ComponentPose=@n16, Alpha=@n15, Translation="0,0,0", Rotation="40,0,0", Scale="1,1,1") {
    rotationMode=BMM_Additive
    rotationSpace=BCS_BoneSpace
    translationMode=BMM_Additive
    translationSpace=BCS_BoneSpace
    alphaBoolBlend {blendOption=Linear}
    alphaScaleBiasClamp {inRange {max=0, min=-100}, outRange {max=0, min=1}}
    boneToModify {boneName=driver_frill_bck_r}
    rotation {pitch=40}
   }
   n18 = "Constraint - Flap_Front_Driven_Bone (PP)"!(ComponentPose=@n17, Alpha=1, ConstraintWeights_0=0.5, ConstraintWeights_1:float=0.5) -> (Pose) {
    alphaBoolBlend {blendOption=Linear}
    boneToModify {boneName=Flap_Front_Driven_Bone}
    constraintSetup [{targetBone {boneName=dyn_frill_frt_l}}, {targetBone {boneName=dyn_frill_frt_r}}]
    constraintWeights [1,1]
   }
   n19 = "Constraint - driver_frill_bck (PP)"!(ComponentPose=@n18, Alpha=1, ConstraintWeights_0=0.5, ConstraintWeights_1:float=0.5) -> (Pose) {
    alphaBoolBlend {blendOption=Linear}
    boneToModify {boneName=driver_frill_bck}
    constraintSetup [{targetBone {boneName=dyn_frill_bck_l}}, {targetBone {boneName=dyn_frill_bck_r}}]
    constraintWeights [1,1]
   }
   n20 = RigidBody(ComponentPose=@n19) {
    bClampLinearTranslationLimitToRefPose=true
    bOverrideWorldGravity=true
    overridePhysicsAsset=&/F_MED_Sheath_Dyn_Skirt_Physics_new
    simulationSpace=BaseBoneSpace
    alphaBoolBlend {blendOption=Linear}
    baseBoneRef {boneName=root}
    componentLinearAccScale=<0.5 0.5 0.55>
    componentLinearVelScale=<0.5 0.5 0.55>
    overrideWorldGravity {z=-980}
   }
   n21 = "Transform (Modify) Bone - Bone: dyn_frill_frt_l"(ComponentPose=@n20, Alpha=1, Translation="0,1,0", Rotation="0,0,0", Scale="1,1,1") {
    rotationMode=BMM_Additive
    rotationSpace=BCS_BoneSpace
    translationMode=BMM_Additive
    translationSpace=BCS_BoneSpace
    alphaBoolBlend {blendOption=Linear}
    alphaScaleBiasClamp {outRange {max=0, min=1}}
    boneToModify {boneName=dyn_frill_frt_l}
    translation {y=1}
   }
   n22 = "Transform (Modify) Bone - Bone: dyn_frill_side_l"(ComponentPose=@n21, Alpha=1, Translation="0,0,-2", Rotation="5,0,0", Scale="1,1,1") {
    rotationMode=BMM_Additive
    rotationSpace=BCS_BoneSpace
    translationMode=BMM_Additive
    translationSpace=BCS_BoneSpace
    alphaBoolBlend {blendOption=Linear}
    alphaScaleBiasClamp {outRange {max=0, min=1}}
    boneToModify {boneName=dyn_frill_side_l}
    rotation {pitch=5}
    translation {z=-2}
   }
   n23 = "Transform (Modify) Bone - Bone: dyn_frill_bck_l"(ComponentPose=@n22, Alpha=1, Translation="0,-1,0", Rotation="0,0,0", Scale="1,1,1") {
    rotationMode=BMM_Additive
    rotationSpace=BCS_BoneSpace
    translationMode=BMM_Additive
    translationSpace=BCS_BoneSpace
    alphaBoolBlend {blendOption=Linear}
    alphaScaleBiasClamp {outRange {max=0, min=1}}
    boneToModify {boneName=dyn_frill_bck_l}
    translation {y=-1}
   }
   n24 = "Transform (Modify) Bone - Bone: dyn_frill_frt_r"(ComponentPose=@n23, Alpha=1, Translation="0,-1,0", Rotation="5,0,0", Scale="1,1,1") {
    rotationMode=BMM_Additive
    rotationSpace=BCS_BoneSpace
    translationMode=BMM_Additive
    translationSpace=BCS_BoneSpace
    alphaBoolBlend {blendOption=Linear}
    alphaScaleBiasClamp {outRange {max=0, min=1}}
    boneToModify {boneName=dyn_frill_frt_r}
    rotation {pitch=5}
    translation {y=-1}
   }
   n25 = "Transform (Modify) Bone - Bone: dyn_frill_side_r"(ComponentPose=@n24, Alpha=1, Translation="0,0,2", Rotation="5,0,0", Scale="1,1,1") {
    rotationMode=BMM_Additive
    rotationSpace=BCS_BoneSpace
    translationMode=BMM_Additive
    translationSpace=BCS_BoneSpace
    alphaBoolBlend {blendOption=Linear}
    alphaScaleBiasClamp {outRange {max=0, min=1}}
    boneToModify {boneName=dyn_frill_side_r}
    rotation {pitch=5}
    translation {z=2}
   }
   n26 = "Transform (Modify) Bone - Bone: dyn_frill_bck_r"(ComponentPose=@n25, Alpha=1, Translation="0,1,0", Rotation="0,0,0", Scale="1,1,1") {
    rotationMode=BMM_Additive
    rotationSpace=BCS_BoneSpace
    translationMode=BMM_Additive
    translationSpace=BCS_BoneSpace
    alphaBoolBlend {blendOption=Linear}
    alphaScaleBiasClamp {outRange {max=0, min=1}}
    boneToModify {boneName=dyn_frill_bck_r}
    translation {y=1}
   }
   n27 = Get SkyDivingAlpha() -> (SkyDivingAlpha:float)
   n28 = "Transform (Modify) Bone - Bone: dyn_frill_bck"(ComponentPose=@n26, Alpha=@n27, Translation="0,-3,0", Rotation="0,0,0", Scale="1,1,1") {
    rotationMode=BMM_Additive
    rotationSpace=BCS_BoneSpace
    translationMode=BMM_Additive
    translationSpace=BCS_BoneSpace
    alphaBoolBlend {blendOption=Linear}
    alphaScaleBiasClamp {bInterpResult=true, outRange {max=0, min=1}}
    boneToModify {boneName=dyn_frill_bck}
    translation {y=-3}
   }
   n29 = "Transform (Modify) Bone - Bone: dyn_frill_bck_l"(ComponentPose=@n28, Alpha=@n27, Translation="0,-3,0", Rotation="0,0,0", Scale="1,1,1") {
    rotationMode=BMM_Additive
    rotationSpace=BCS_BoneSpace
    translationMode=BMM_Additive
    translationSpace=BCS_BoneSpace
    alphaBoolBlend {blendOption=Linear}
    alphaScaleBiasClamp {bInterpResult=true, outRange {max=0, min=1}}
    boneToModify {boneName=dyn_frill_bck_l}
    translation {y=-3}
   }
   n30 = "Transform (Modify) Bone - Bone: dyn_frill_bck_r"(ComponentPose=@n29, Alpha=@n27, Translation="0,3,0", Rotation="0,0,0", Scale="1,1,1") {
    rotationMode=BMM_Additive
    rotationSpace=BCS_BoneSpace
    translationMode=BMM_Additive
    translationSpace=BCS_BoneSpace
    alphaBoolBlend {blendOption=Linear}
    alphaScaleBiasClamp {bInterpResult=true, outRange {max=0, min=1}}
    boneToModify {boneName=dyn_frill_bck_r}
    translation {y=3}
   }
   n31 = "Constraint - dyn_wire_front (P)"(ComponentPose=@n30, Alpha=1, ConstraintWeights_0=1) {
    alphaBoolBlend {blendOption=Linear}
    boneToModify {boneName=dyn_wire_front}
    constraintSetup [{targetBone {boneName=dyn_frill_frt_r}}]
    constraintWeights [1]
   }
   n32 = "Constraint - dyn_wire_back (P)"(ComponentPose=@n31, Alpha=1, ConstraintWeights_0=1) {
    alphaBoolBlend {blendOption=Linear}
    boneToModify {boneName=dyn_wire_back}
    constraintSetup [{targetBone {boneName=dyn_frill_side_r}}]
    constraintWeights [1]
   }
   n33 = "Transform (Modify) Bone - Bone: dyn_wire_front"(ComponentPose=@n32, Alpha=Get LegUpRightAlpha(), Translation="0,-5,-1", Rotation="0,0,0", Scale="1,1,1") {
    rotationMode=BMM_Additive
    rotationSpace=BCS_BoneSpace
    translationMode=BMM_Additive
    translationSpace=BCS_BoneSpace
    alphaBoolBlend {blendOption=Linear}
    alphaScaleBiasClamp {bMapRange=true, inRange {max=0, min=-100}, outRange {max=0, min=1}}
    boneToModify {boneName=dyn_wire_front}
    translation {y=-5, z=-1}
   }
   n34 = "Transform (Modify) Bone - Bone: dyn_wire_front"(ComponentPose=@n33, Alpha=1, Translation="0,-3,0", Rotation="0,0,0", Scale="1,1,1") {
    rotationMode=BMM_Additive
    rotationSpace=BCS_BoneSpace
    translationMode=BMM_Additive
    translationSpace=BCS_BoneSpace
    alphaBoolBlend {blendOption=Linear}
    alphaScaleBiasClamp {inRange {max=0, min=-100}, outRange {max=0, min=1}}
    boneToModify {boneName=dyn_wire_front}
    translation {y=-3}
   }
   n35 = Anim Dynamics - Bone: dyn_wire(ComponentPose=@n34, Alpha=1, ExternalForce="0,0,0") {
    angularDampingOverride=1
    angularSpringConstant=1500
    bAngularSpring=true
    bGravityOverrideInSimSpace=true
    bOverrideAngularDamping=true
    gravityScale=0.35
    simulationSpace=RootRelative
    alphaBoolBlend {blendOption=Linear}
    boundBone {boneName=dyn_wire}
    componentLinearAccScale=<0.2 0.2 0.4>
    componentLinearVelScale=<0.2 0.2 0.4>
    physicsBodyDefinitions [{
      sphereCollisionRadius=0
      boundBone {boneName=dyn_wire}
      boxExtents=<10 20 20>
      constraintSetup {angularLimitsMax=<0 50 0>, angularLimitsMin=<0 0 0>, angularTarget=<0 0 0>}
      localJointOffset=<10 0 0>
     }]
   }
   n36 = Anim Dynamics - Bone: dyn_wire_front(ComponentPose=@n35, Alpha=1, ExternalForce="0,0,0") {
    angularDampingOverride=1
    angularSpringConstant=1500
    bAngularSpring=true
    bGravityOverrideInSimSpace=true
    bOverrideAngularDamping=true
    gravityScale=0.35
    simulationSpace=RootRelative
    alphaBoolBlend {blendOption=Linear}
    boundBone {boneName=dyn_wire_front}
    componentLinearAccScale=<0.2 0.2 0.3>
    componentLinearVelScale=<0.2 0.2 0.3>
    physicsBodyDefinitions [{
      sphereCollisionRadius=0
      boundBone {boneName=dyn_wire_front}
      boxExtents=<10 20 20>
      constraintSetup {angularLimitsMax=<0 0 0>, angularLimitsMin=<0 0 -30>, angularTarget=<0 0 0>, linearAxesMin=<0 -3 0>}
      localJointOffset=<10 0 0>
     }]
   }
   n37 = Anim Dynamics - Bone: dyn_wire_back(ComponentPose=@n36, Alpha=1, ExternalForce="0,0,0") {
    angularDampingOverride=1
    angularSpringConstant=1500
    bAngularSpring=true
    bGravityOverrideInSimSpace=true
    bOverrideAngularDamping=true
    gravityScale=0.35
    simulationSpace=RootRelative
    alphaBoolBlend {blendOption=Linear}
    boundBone {boneName=dyn_wire_back}
    componentLinearAccScale=<0.2 0.2 0.3>
    componentLinearVelScale=<0.2 0.2 0.3>
    physicsBodyDefinitions [{
      sphereCollisionRadius=0
      boundBone {boneName=dyn_wire_back}
      boxExtents=<10 20 20>
      constraintSetup {angularLimitsMax=<0 0 0>, angularLimitsMin=<0 0 -30>, angularTarget=<0 0 0>, linearAxesMin=<0 -3 0>}
      localJointOffset=<10 0 0>
     }]
   }
   n38 = "Transform (Modify) Bone - Bone: dyn_belt_Base"(ComponentPose=@n37, Alpha=Get JumpAlpha() -> (JumpAlpha:float), Translation="0,0,1", Rotation="0,-15,0", Scale="1,1,1") {
    rotationMode=BMM_Additive
    rotationSpace=BCS_BoneSpace
    translationMode=BMM_Additive
    translationSpace=BCS_BoneSpace
    alphaBoolBlend {blendOption=Linear}
    alphaScaleBiasClamp {inRange {max=0, min=-100}, outRange {max=0, min=1}}
    boneToModify {boneName=dyn_belt_Base}
    rotation {yaw=-15}
    translation {z=1}
   }
   n39 = "Constraint - dyn_hipPouchMain_Base (P)"(ComponentPose=@n38, Alpha=1, ConstraintWeights_0=1) {
    alphaBoolBlend {blendOption=Linear}
    boneToModify {boneName=dyn_hipPouchMain_Base}
    constraintSetup [{targetBone {boneName=dyn_beltEND}}]
    constraintWeights [1]
   }
   n40 = "Transform (Modify) Bone - Bone: dyn_hipPouchMain_Base"(ComponentPose=@n39, Alpha=1, Translation="-4,-6,2", Rotation="10,-50,-5", Scale="1,1,1") {
    rotationMode=BMM_Additive
    rotationSpace=BCS_BoneSpace
    translationMode=BMM_Additive
    translationSpace=BCS_BoneSpace
    alphaBoolBlend {blendOption=Linear}
    alphaScaleBiasClamp {inRange {max=0, min=-100}, outRange {max=0, min=1}}
    boneToModify {boneName=dyn_hipPouchMain_Base}
    rotation=<rot 10 -50 -5>
    translation=<-4 -6 2>
   }
   n41 = "Transform (Modify) Bone - Bone: dyn_hipPouchMain_Base"(ComponentPose=@n40, Alpha=Get LegUpLeftAlpha(), Translation="-1.5,-3,1", Rotation="0,-25,0", Scale="1,1,1") {
    rotationMode=BMM_Additive
    rotationSpace=BCS_BoneSpace
    translationMode=BMM_Additive
    translationSpace=BCS_BoneSpace
    alphaBoolBlend {blendOption=Linear}
    alphaScaleBiasClamp {bMapRange=true, inRange {max=0, min=-100}, outRange {max=0, min=1}}
    boneToModify {boneName=dyn_hipPouchMain_Base}
    rotation {yaw=-25}
    translation=<-1.5 -3 1>
   }
   n42 = RigidBody(ComponentPose=@n41) {
    bClampLinearTranslationLimitToRefPose=true
    bOverrideWorldGravity=true
    overridePhysicsAsset=&/F_MED_Sheath_Dyn_Acc_Physics
    simulationSpace=BaseBoneSpace
    alphaBoolBlend {blendOption=Linear}
    baseBoneRef {boneName=root}
    componentAppliedLinearAccClamp=<3000 3000 3000>
    componentLinearAccScale=<0.5 0.5 0.5>
    componentLinearVelScale=<0.25 0.255 0.3>
    overrideWorldGravity {z=-980}
   }
   n43 = "Constraint - dyn_neckZipper_Base (PPPPPP)"!(
    ComponentPose=@n42
    Alpha=1
    ConstraintWeights_0=0.0716
    ConstraintWeights_1:float=0.0011
    ConstraintWeights_2:float=0.2473
    ConstraintWeights_3:float=0.0007
    ConstraintWeights_4:float=0.2915
    ConstraintWeights_5:float=0.3878
   ) -> (Pose) {
    alphaBoolBlend {blendOption=Linear}
    boneToModify {boneName=dyn_neckZipper_Base}
    constraintSetup [
     {targetBone {boneName=head}}
     {targetBone {boneName=clavicle_l}}
     {targetBone {boneName=spine_05}}
     {targetBone {boneName=clavicle_r}}
     {targetBone {boneName=neck_02}}
     {targetBone {boneName=neck_01}}
    ]
    constraintWeights [1,1,1,1,1,1]
   }
   n44 = Control Rig!(Source=Component To Local(ComponentPose=@n43), Alpha, crouching:double=Get CrouchAlpha() -> (CrouchAlpha:float)) -> (Pose) {
    alphaBoolBlend {blendOption=Linear}
    controlRigAssetReference {blueprintRigClass=&/F_MED_Sheath_CtrlRig.F_MED_Sheath_CtrlRig_C}
   }
   Output Pose(Result=@n44) {layerGroup=None}
  ]}
 EventGraph {links [], nodes []}
]
variables []
